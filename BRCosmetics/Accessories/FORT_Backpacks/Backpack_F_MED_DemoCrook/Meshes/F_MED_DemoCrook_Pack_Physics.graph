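# Physics asset for the DemoCrook backpack cosmetic, giving capsule collision bodies and joints to its hotdog-top and fuse bones.
# A three-link hotdog-top chain and a two-link fuse chain both branch from dyn_hotdog_top_MSTR, with tight swing and twist limits.
# The pack and that master bone each attach separately to root with free angular motion.
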PhysicsAsset /BRCosmetics/Accessories/FORT_Backpacks/Backpack_F_MED_DemoCrook/Meshes/F_MED_DemoCrook_Pack_Physics
bodies {
 dyn_fuse_01 [{shapeType=Capsule, length=1.6, radius=0.6, shapeName=dyn_fuse_01_capsule, center=<1.4656 0 0>, rotation=<rot -90 0 0>}]
 dyn_fuse_02 [{shapeType=Capsule, length=0, radius=3, shapeName=dyn_fuse_02_capsule, center=<3.1242 0 0>, rotation=<rot -90 0 0>}]
 dyn_hotdog_top_01 [
  {shapeType=Capsule, length=4.5, radius=3.5, shapeName=dyn_hotdog_top_01_capsule, center=<6.1181 0.9405 -0.3313>, rotation=<rot 90 0 0>}
  {shapeType=Capsule, length=3, radius=3.5, shapeName=dyn_hotdog_top_01_capsule, center=<3.0374 -2.8689 1.3902>}
 ]
 dyn_hotdog_top_02 [{shapeType=Capsule, length=1.5, radius=3, shapeName=dyn_hotdog_top_02_capsule, center=<4.1598 0 0>, rotation=<rot 90 0 0>}]
 dyn_hotdog_top_03 [{shapeType=Capsule, length=1, radius=4, shapeName=dyn_hotdog_top_03_capsule, center=<4.7649 0 0>, rotation=<rot 90 0 0>}]
 dyn_hotdog_top_MSTR [{shapeType=Capsule, length=0.505, radius=0.505, shapeName=dyn_hotdog_top_MSTR_capsule, rotation=<rot 0 -90 0>}]
 pack [{shapeType=Capsule, length=0.505, radius=0.505, shapeName=pack_capsule}]
 root [{shapeType=Capsule, length=1, radius=1}]
}
constraints [
 {bone1Name=dyn_hotdog_top_01, bone2Name=dyn_hotdog_top_MSTR, swing1LimitDegrees=10, swing2LimitDegrees=10, twistLimitDegrees=3}
 {bone1Name=dyn_hotdog_top_02, bone2Name=dyn_hotdog_top_01, swing1LimitDegrees=10, swing2LimitDegrees=10, twistLimitDegrees=3}
 {bone1Name=dyn_hotdog_top_03, bone2Name=dyn_hotdog_top_02, swing1LimitDegrees=10, swing2LimitDegrees=10, twistLimitDegrees=3}
 {bone1Name=dyn_fuse_02, bone2Name=dyn_fuse_01, swing1LimitDegrees=15, swing2LimitDegrees=15, twistLimitDegrees=2.5}
 {
  bone1Name=pack
  bone2Name=root
  swing1LimitDegrees=45
  swing1Motion=Free
  swing2LimitDegrees=45
  swing2Motion=Free
  twistLimitDegrees=45
  twistMotion=Free
 }
 {bone1Name=dyn_fuse_01, bone2Name=dyn_hotdog_top_MSTR, swing1LimitDegrees=11, swing2LimitDegrees=11, twistLimitDegrees=0}
 {
  bone1Name=dyn_hotdog_top_MSTR
  bone2Name=root
  swing1LimitDegrees=45
  swing1Motion=Free
  swing2LimitDegrees=45
  swing2Motion=Free
  twistLimitDegrees=45
  twistMotion=Free
 }
]
