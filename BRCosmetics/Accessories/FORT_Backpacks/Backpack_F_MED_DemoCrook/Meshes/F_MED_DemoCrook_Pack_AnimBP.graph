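# Animation Blueprint for the DemoCrook cosmetic backpack, adding physics sway and state-based offsets to the pack bone.
# Rigid-body simulation in pack-bone space runs on the cached input pose.
# While skydiving, it blends to a second simulated branch offset by owner spine-bend values computed in the EventGraph, plus skydive and parachute nudges.
# A final front-end offset follows; the deprecated gravity-override update sets a variable the AnimGraph never reads.
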AnimBlueprint /BRCosmetics/Accessories/FORT_Backpacks/Backpack_F_MED_DemoCrook/Meshes/F_MED_DemoCrook_Pack_AnimBP
target_skeleton=&/F_MED_DemoCrook_Pack_Skeleton
graphs [
 AnimGraph {nodes [
   n1 = RigidBody(ComponentPose=Local To Component(LocalPose=Use cached pose 'cache pose'())) {
    bOverrideWorldGravity=true
    overridePhysicsAsset=&/F_MED_DemoCrook_Pack_Physics
    simulationSpace=BaseBoneSpace
    alphaBoolBlend {blendOption=Linear}
    baseBoneRef {boneName=pack}
    componentAppliedLinearAccClamp=<1000 1000 1000>
    componentLinearAccScale=<0.6 0.6 0.6>
    componentLinearVelScale=<0.4 0.4 0.4>
    overrideWorldGravity {z=-980}
   }
   n2 = Component To Local(ComponentPose=@n1)
   n3 = RigidBody(ComponentPose=Local To Component(LocalPose=Use cached pose 'cache pose'())) {
    bOverrideWorldGravity=true
    overridePhysicsAsset=&F_MED_DemoCrook_Pack_Physics
    simulationSpace=BaseBoneSpace
    alphaBoolBlend {blendOption=Linear}
    baseBoneRef {boneName=pack}
    componentAppliedLinearAccClamp=<1000 1000 1000>
    componentLinearAccScale=<0.6 0.6 0.6>
    componentLinearVelScale=<0.4 0.4 0.4>
    overrideWorldGravity {z=-980}
   }
   n4 = "Transform (Modify) Bone - Bone: pack"(
    ComponentPose=@n3
    Translation=Get Backpack Translation() -> (Backpack Translation:Vector)
    Rotation=Get Backpack Rotate() -> (Backpack Rotate:Rotator)
   ) {
    rotationMode=BMM_Additive
    rotationSpace=BCS_BoneSpace
    scaleMode=BMM_Additive
    scaleSpace=BCS_BoneSpace
    translationMode=BMM_Additive
    translationSpace=BCS_BoneSpace
    alphaBoolBlend {blendOption=Linear}
    boneToModify {boneName=pack}
   }
   n5 = "Transform (Modify) Bone - Bone: pack"!(ComponentPose=@n4, bAlphaBoolEnabled:bool=Get bIsSkyDiving(), Translation="0,-1,0", Rotation="-0.000,0.000,10.000", Scale) -> (Pose) {
    alphaInputType=Bool
    rotationMode=BMM_Additive
    rotationSpace=BCS_BoneSpace
    scaleMode=BMM_Additive
    scaleSpace=BCS_BoneSpace
    translationMode=BMM_Additive
    translationSpace=BCS_BoneSpace
    alphaBoolBlend {blendOption=Linear}
    boneToModify {boneName=pack}
    rotation {roll=10}
    translation {y=-1}
   }
   n6 = "Transform (Modify) Bone - Bone: pack"(ComponentPose=@n5, bAlphaBoolEnabled=Get bIsParachuteOpened(), Translation="0.000,2.220,-0.000", Rotation="0,0,0") {
    alphaInputType=Bool
    rotationMode=BMM_Additive
    rotationSpace=BCS_BoneSpace
    scaleMode=BMM_Additive
    scaleSpace=BCS_BoneSpace
    translationMode=BMM_Additive
    translationSpace=BCS_BoneSpace
    alphaBoolBlend {blendOption=Linear}
    boneToModify {boneName=pack}
    translation {y=2.22}
   }
   n7 = Component To Local(ComponentPose=@n6)
   n8 = Get bIsSkyDiving()
   n9 = "Transform (Modify) Bone - Bone: pack"(
    ComponentPose=Local To Component(LocalPose=Blend Poses by bool(BlendPose_0=@n7, BlendPose_1=@n2, BlendTime_0=0.5, bActiveValue=@n8) {blendType=Linear})
    bAlphaBoolEnabled=Get bIsInFrontEnd()
    Translation="0,1.3,0"
    Rotation="0,0,0"
   ) {
    alphaInputType=Bool
    translationMode=BMM_Additive
    translationSpace=BCS_BoneSpace
    alphaBoolBlend {blendOption=Linear}
    boneToModify {boneName=pack}
    translation {y=1.3}
   }
   n10 = Component To Local(ComponentPose=@n9)
   Output Pose(Result=Component To Local(ComponentPose=Local To Component(LocalPose=@n10))) {layerGroup=None}
   Save cached pose 'cache pose'(Pose=Input Pose()) {cacheName=cache pose}
  ]}
 EventGraph {nodes [
   Event BlueprintUpdateAnimation()
   > n1 = Sequence()
   > Gravity Override State Function DEPRECATED(CharacterPartAnimInstance=Self-Reference(), GameTimeInSeconds=GetGameTimeInSeconds(), GravityOverrideParams=Get Gravity Override Params())
   > Set Gravity Override(Gravity Override="0, 0, 0")
   n2 = GetOwnerSkeletalMeshComponent()
   n3 = BreakRotator(InRot=BreakTransform(InTransform=GetDeltaTransformFromRefPose(self=@n2, BoneName=spine_05, BaseName=spine_01)).Rotation)
   Set Backpack Rotate(
    execute=@n1.then_1
    Backpack Rotate:Rotator=MakeRotator(Roll=MapRangeClamped(Value=@n3.Yaw, InRangeA=-12, InRangeB=-30, OutRangeA=3, OutRangeB=25))
   ) -> (Output_Get:Rotator)
   > Set Backpack Translation(Backpack Translation:Vector=MakeVector(Y=MapRangeClamped(Value=@n3.Yaw, InRangeA=-12, InRangeB=-30, OutRangeB=-8))) -> (Output_Get:Vector)
   TryGetPawnOwner()
  ]}
]
variables [
 Backpack Rotate {type=Rotator}
 Backpack Translation {type=Vector}
 Gravity Override {type=Vector}
 Gravity Override Params {type=GravityOverrideParamsStruct}
 Is In Front End {type=boolean}
 Is Parachute Opened {type=boolean}
 Is Sky Diving {type=boolean}
]
